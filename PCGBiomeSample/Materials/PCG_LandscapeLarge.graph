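# Masked, Nanite-enabled landscape material in the PCG biome sample that shades terrain as three flat-coloured layers: Land, Shore and Mountain.
# A weight blend of the layer colours feeds base color, with constant roughness and ambient occlusion and the landscape visibility mask driving opacity mask.
# The same layers feed a grass output with no grass types assigned.
# A runtime virtual texture output writes only world height, from world position Z.
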Material /PCGBiomeSample/Materials/PCG_LandscapeLarge
graph {
 properties {bUseMaterialAttributes=true, bUsedWithNanite=true, blendMode=BLEND_Masked, pixelDepthOffsetMode=PDOM_Legacy, usedShadingModels=Default Lit}
 nodes [
  MaterialExpressionLandscapeLayerSample_0 = LandscapeLayerSample() {parameterName=Land}
  MaterialExpressionLandscapeLayerSample_2 = LandscapeLayerSample() {parameterName=Shore}
  MaterialExpressionLandscapeLayerSample_1 = LandscapeLayerSample() {parameterName=Mountain}
  LandscapeGrassOutput!(
   Land=@MaterialExpressionLandscapeLayerSample_0
   Shore=@MaterialExpressionLandscapeLayerSample_2
   Mountain=@MaterialExpressionLandscapeLayerSample_1
  ) {grassTypes [
    Land {
     grassType=null
     input {
      expression=&self:MaterialExpressionLandscapeLayerSample_0
      inputName=None
      mask=0
      maskA=0
      maskB=0
      maskG=0
      maskR=0
      outputIndex=0
     }
    }
    Shore {
     grassType=null
     input {
      expression=&self:MaterialExpressionLandscapeLayerSample_2
      inputName=None
      mask=0
      maskA=0
      maskB=0
      maskG=0
      maskR=0
      outputIndex=0
     }
    }
    Mountain {
     grassType=null
     input {
      expression=&self:MaterialExpressionLandscapeLayerSample_1
      inputName=None
      mask=0
      maskA=0
      maskB=0
      maskG=0
      maskR=0
      outputIndex=0
     }
    }
   ]}
  MaterialExpressionConstant4Vector_5 = Constant4Vector() {constant=<rgba 0.472 0.1983 0.0902 0>}
  MaterialExpressionConstant4Vector_1 = Constant4Vector() {constant=<rgba 0.6528 0.4081 0.1904 0>}
  MaterialExpressionConstant4Vector_6 = Constant4Vector() {constant=<rgba 0.387 0.3486 0.3009 1>}
  n1 = LandscapeLayerBlend!(
   Layer Land=@MaterialExpressionConstant4Vector_5.RGBA
   Layer Shore=@MaterialExpressionConstant4Vector_1.RGBA
   Layer Mountain=@MaterialExpressionConstant4Vector_6.RGBA
  ) {layers [
    {
     blendType=LB_WeightBlend
     constHeightInput=0
     layerName=Land
     previewWeight=0
     constLayerInput=<0 0 0>
     heightInput {expression=null, inputName=None, mask=0, maskA=0, maskB=0, maskG=0, maskR=0, outputIndex=0}
     layerInput {
      expression=&self:MaterialExpressionConstant4Vector_5
      inputName=None
      mask=1
      maskA=1
      maskB=1
      maskG=1
      maskR=1
      outputIndex=0
     }
    }
    {
     blendType=LB_WeightBlend
     constHeightInput=0
     layerName=Shore
     previewWeight=0
     constLayerInput=<0 0 0>
     heightInput {expression=null, inputName=None, mask=0, maskA=0, maskB=0, maskG=0, maskR=0, outputIndex=0}
     layerInput {
      expression=&self:MaterialExpressionConstant4Vector_1
      inputName=None
      mask=1
      maskA=1
      maskB=1
      maskG=1
      maskR=1
      outputIndex=0
     }
    }
    {
     blendType=LB_WeightBlend
     constHeightInput=0
     layerName=Mountain
     previewWeight=0
     constLayerInput=<0 0 0>
     heightInput {expression=null, inputName=None, mask=0, maskA=0, maskB=0, maskG=0, maskR=0, outputIndex=0}
     layerInput {
      expression=&self:MaterialExpressionConstant4Vector_6
      inputName=None
      mask=1
      maskA=1
      maskB=1
      maskG=1
      maskR=1
      outputIndex=0
     }
    }
   ]}
  n2 = Constant() {r=1}
  MakeMaterialAttributes!(
   BaseColor=@n1
   Metallic
   Specular
   Roughness=@n2
   Anisotropy
   EmissiveColor
   Opacity
   OpacityMask=LandscapeVisibilityMask()
   Normal
   Tangent
   WorldPositionOffset
   SubsurfaceColor
   ClearCoat
   ClearCoatRoughness
   AmbientOcclusion=@n2
   Refraction
   CustomizedUVs_0
   CustomizedUVs_1
   CustomizedUVs_2
   CustomizedUVs_3
   CustomizedUVs_4
   CustomizedUVs_5
   CustomizedUVs_6
   CustomizedUVs_7
   PixelDepthOffset
   ShadingModel
   Displacement
  )
  RuntimeVirtualTextureOutput!(BaseColor, Specular, Roughness, Normal, WorldHeight=WorldPosition().Z, Opacity, Mask, Displacement, Mask4)
 ]
}
