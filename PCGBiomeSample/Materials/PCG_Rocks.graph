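# PCG_Rocks is a simple Default Lit surface material, named for rocks in the PCG Biome sample.
# Its graph holds only a dark grey vector parameter and a scalar constant, with no textures and no visible wiring to material outputs.
# Usage flags enable it on instanced static meshes and Nanite meshes.
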Material /PCGBiomeSample/Materials/PCG_Rocks
graph {
 properties {bUsedWithInstancedStaticMeshes=true, bUsedWithNanite=true, pixelDepthOffsetMode=PDOM_Legacy, usedShadingModels=Default Lit}
 nodes [
  Constant() {r=1}
  VectorParameter() {defaultValue=<rgba 0.1684 0.1684 0.1684 0>}
 ]
}
