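# PCG_Trunk_01 is a Default Lit material in the PCG Biome Sample, apparently for tree trunks going by its name.
# It holds a near-black BaseColor vector parameter and two scalar constants, with no textures and no visible wiring to material outputs.
# It is flagged for use with instanced static meshes and Nanite.
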Material /PCGBiomeSample/Materials/PCG_Trunk_01
graph {
 properties {bUsedWithInstancedStaticMeshes=true, bUsedWithNanite=true, pixelDepthOffsetMode=PDOM_Legacy, usedShadingModels=Default Lit}
 nodes [
  Constant() {r=0.8}
  Constant()
  VectorParameter() {parameterName=BaseColor, defaultValue=<rgba 0.026 0.024 0.023 0>}
 ]
}
