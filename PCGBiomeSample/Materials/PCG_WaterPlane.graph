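# PCG_WaterPlane is a simple Single Layer Water material in the PCG Biome sample, named for use on a water plane.
# A Single Layer Water output node takes constant scattering, absorption, phase and color-scale-behind-water values to set the water's look.
# The graph uses only constants, with no textures, parameters, normals or animation.
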Material /PCGBiomeSample/Materials/PCG_WaterPlane
graph {
 properties {pixelDepthOffsetMode=PDOM_Legacy, shadingModel=MSM_SingleLayerWater, usedShadingModels=SingleLayerWater}
 nodes [
  Constant4Vector() {constant=<rgba 0.0013 0.0644 0.1285 0>}
  Constant() {r=0.1}
  SingleLayerWaterMaterialOutput!(
   ScatteringCoefficients=Constant() {r=0.0012}
   AbsorptionCoefficients=Constant() {r=0.2}
   PhaseG=Constant() {r=0.25}
   ColorScaleBehindWater=Constant4Vector() {constant=<rgba 0 0.0228 0.1806 0>}.RGBA
  )
 ]
}
